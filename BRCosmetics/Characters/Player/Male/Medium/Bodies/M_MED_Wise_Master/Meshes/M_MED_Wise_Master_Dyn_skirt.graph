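# Physics asset for the dynamic rear skirt (loincloth) bones of the male medium Wise Master player body.
# Center, left and right five-bone chains hang from pelvis-linked MSTR bodies with limited swing and tight twist, using capsules, or boxes on lower side segments.
# Angularly free constraints cross-link both side chains' second segments to the center's third.
# Pelvis and leg capsules, plus a large sphere mostly below root, provide body and ground-level collision shapes.
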PhysicsAsset /BRCosmetics/Characters/Player/Male/Medium/Bodies/M_MED_Wise_Master/Meshes/M_MED_Wise_Master_Dyn_skirt
bodies {
 calf_l [
  {shapeType=Capsule, length=18.8087, radius=13.7759, center=<-10.4025 -1.5669 -1.2766>, rotation=<rot 90 0 -7.7167>}
  {shapeType=Capsule, length=29.7058, radius=8.1079, center=<-35.0988 -0.1083 -0.6595>, rotation=<rot 90 0 3.9547>}
 ]
 calf_r [
  {shapeType=Capsule, length=20.3087, radius=13.2759, center=<11.1046 1.6621 1.2766>, rotation=<rot 90 83.2785 85.5619>}
  {shapeType=Capsule, length=29.2058, radius=8.6079, center=<34.9565 -1.951 0.6595>, rotation=<rot 90 180 -176.0453>}
 ]
 dyn_loincloth_bk_01 [{shapeType=Capsule, length=10.5, radius=8.25, shapeName=dyn_loincloth_bk_01_capsule, center=<6 0 7.5>, rotation=<rot 0 0 89.9999>}]
 dyn_loincloth_bk_02 [{shapeType=Capsule, length=15.5, radius=9.5, shapeName=dyn_loincloth_bk_02_capsule, center=<8 0 7.5>, rotation=<rot 0 0 89.9999>}]
 dyn_loincloth_bk_03 [{
   shapeType=Capsule
   length=22.4482
   radius=7.5518
   shapeName=dyn_loincloth_bk_03_capsule
   center=<7.9572 0.045 0.5139>
   rotation=<rot -0.0003 0 95>
  }]
 dyn_loincloth_bk_04 [{
   shapeType=Capsule
   length=22.4482
   radius=7.5518
   shapeName=dyn_loincloth_bk_04_capsule
   center=<7.9572 0.045 0.5139>
   rotation=<rot -0.0003 0 95>
  }]
 dyn_loincloth_bk_05 [{shapeType=Capsule, length=22.4482, radius=7.5518, shapeName=dyn_loincloth_bk_05_capsule, rotation=<rot -0.0003 0 95>}]
 dyn_loincloth_bk_MSTR [{shapeType=Capsule, length=1, radius=1, rotation=<rot 90 0 0>}]
 dyn_loincloth_bk_l_01 [{shapeType=Capsule, length=10.5, radius=8.25, shapeName=dyn_loincloth_bk_l_01_capsule, center=<6 0 7.5>, rotation=<rot 0 0 89.9999>}]
 dyn_loincloth_bk_l_02 [{shapeType=Capsule, length=11.75, radius=9, shapeName=dyn_loincloth_bk_l_02_capsule, center=<8.2411 0 6.9994>, rotation=<rot 0 0 89.9999>}]
 dyn_loincloth_bk_l_03 [{shapeType=Box, extentX=18.4353, extentY=16.5943, extentZ=6.5538, center=<8.4328 0.0001 0>}]
 dyn_loincloth_bk_l_04 [{shapeType=Box, extentX=20.3661, extentY=28.3474, extentZ=7.3298, center=<9.2802 0.0001 -0.0001>}]
 dyn_loincloth_bk_l_05 [{shapeType=Box, extentX=16.3749, extentY=28.3474, extentZ=7.3298, center=<8.2554 0.0001 0.0001>}]
 dyn_loincloth_bk_l_MSTR [{shapeType=Capsule, length=1, radius=1, rotation=<rot 90 0 0>}]
 dyn_loincloth_bk_r_01 [{shapeType=Capsule, length=10.5, radius=8.25, shapeName=dyn_loincloth_bk_r_01_capsule, center=<6 0 7.5>, rotation=<rot 0 0 89.9999>}]
 dyn_loincloth_bk_r_02 [{shapeType=Capsule, length=11.75, radius=9, shapeName=dyn_loincloth_bk_r_02_capsule, center=<8.2411 0 6.9994>, rotation=<rot 0 0 89.9999>}]
 dyn_loincloth_bk_r_03 [{shapeType=Box, extentX=18.4353, extentY=16.5943, extentZ=5.5538, center=<8.4328 0.0001 0>}]
 dyn_loincloth_bk_r_04 [{shapeType=Box, extentX=20.3661, extentY=25.3474, extentZ=6.3298, center=<9.2802 0.0001 -0.0001>}]
 dyn_loincloth_bk_r_05 [{shapeType=Box, extentX=16.3749, extentY=25.3474, extentZ=5.3298, center=<8.2554 0.0001 -0.9657>}]
 dyn_loincloth_bk_r_MSTR [{shapeType=Capsule, length=1, radius=1, rotation=<rot 90 0 0>}]
 foot_l [{shapeType=Capsule, length=25.7551, radius=7.6131, center=<-6.6239 10.2131 0.2607>, rotation=<rot -2.184 -174.3694 -87.6997>}]
 foot_r [{shapeType=Capsule, length=25.7551, radius=7.6131, center=<6.6239 -10.2131 -0.2607>, rotation=<rot -2.184 -174.3694 -87.6997>}]
 pelvis [{shapeType=Capsule, length=9.7536, radius=9.467}]
 root [{shapeType=Sphere, radius=150, shapeName=root_sphere, center=<0 0 -140.2227>}]
 thigh_l [{shapeType=Capsule, length=24.1732, radius=11.054, center=<-19.8402 0.2003 -1.3676>, rotation=<rot 86.6266 -180 -180>}]
 thigh_r [{shapeType=Capsule, length=24.6732, radius=10.554, center=<19.8403 -0.2003 1.3681>, rotation=<rot 86.6266 -180 -180>}]
}
constraints [
 {bone1Name=dyn_loincloth_bk_l_01, bone2Name=dyn_loincloth_bk_l_MSTR, swing1LimitDegrees=45, swing2LimitDegrees=45, twistLimitDegrees=10}
 {bone1Name=dyn_loincloth_bk_l_02, bone2Name=dyn_loincloth_bk_l_01, swing1LimitDegrees=45, swing2LimitDegrees=80, twistLimitDegrees=10}
 {bone1Name=dyn_loincloth_bk_l_03, bone2Name=dyn_loincloth_bk_l_02, swing1LimitDegrees=45, swing2LimitDegrees=80, twistLimitDegrees=10}
 {bone1Name=dyn_loincloth_bk_l_04, bone2Name=dyn_loincloth_bk_l_03, swing1LimitDegrees=45, swing2LimitDegrees=80, twistLimitDegrees=10}
 {bone1Name=dyn_loincloth_bk_l_05, bone2Name=dyn_loincloth_bk_l_04, swing1LimitDegrees=45, swing2LimitDegrees=80, twistLimitDegrees=10}
 {bone1Name=dyn_loincloth_bk_01, bone2Name=dyn_loincloth_bk_MSTR, swing1LimitDegrees=45, swing2LimitDegrees=45, twistLimitDegrees=10}
 {bone1Name=dyn_loincloth_bk_02, bone2Name=dyn_loincloth_bk_01, swing1LimitDegrees=45, swing2LimitDegrees=80, twistLimitDegrees=10}
 {bone1Name=dyn_loincloth_bk_03, bone2Name=dyn_loincloth_bk_02, swing1LimitDegrees=45, swing2LimitDegrees=80, twistLimitDegrees=10}
 {bone1Name=dyn_loincloth_bk_04, bone2Name=dyn_loincloth_bk_03, swing1LimitDegrees=45, swing2LimitDegrees=80, twistLimitDegrees=10}
 {bone1Name=dyn_loincloth_bk_05, bone2Name=dyn_loincloth_bk_04, swing1LimitDegrees=45, swing2LimitDegrees=80, twistLimitDegrees=10}
 {bone1Name=dyn_loincloth_bk_r_01, bone2Name=dyn_loincloth_bk_r_MSTR, swing1LimitDegrees=45, swing2LimitDegrees=45, twistLimitDegrees=10}
 {bone1Name=dyn_loincloth_bk_r_02, bone2Name=dyn_loincloth_bk_r_01, swing1LimitDegrees=45, swing2LimitDegrees=80, twistLimitDegrees=10}
 {bone1Name=dyn_loincloth_bk_r_03, bone2Name=dyn_loincloth_bk_r_02, swing1LimitDegrees=45, swing2LimitDegrees=80, twistLimitDegrees=10}
 {bone1Name=dyn_loincloth_bk_r_04, bone2Name=dyn_loincloth_bk_r_03, swing1LimitDegrees=45, swing2LimitDegrees=80, twistLimitDegrees=10}
 {bone1Name=dyn_loincloth_bk_r_05, bone2Name=dyn_loincloth_bk_r_04, swing1LimitDegrees=45, swing2LimitDegrees=80, twistLimitDegrees=10}
 {
  bone1Name=calf_l
  bone2Name=thigh_l
  swing1LimitDegrees=45
  swing1Motion=Free
  swing2LimitDegrees=45
  swing2Motion=Free
  twistLimitDegrees=45
  twistMotion=Free
 }
 {
  bone1Name=calf_r
  bone2Name=thigh_r
  swing1LimitDegrees=45
  swing1Motion=Free
  swing2LimitDegrees=45
  swing2Motion=Free
  twistLimitDegrees=45
  twistMotion=Free
 }
 {
  bone1Name=thigh_l
  bone2Name=pelvis
  swing1LimitDegrees=45
  swing1Motion=Free
  swing2LimitDegrees=45
  swing2Motion=Free
  twistLimitDegrees=45
  twistMotion=Free
 }
 {
  bone1Name=thigh_r
  bone2Name=pelvis
  swing1LimitDegrees=45
  swing1Motion=Free
  swing2LimitDegrees=45
  swing2Motion=Free
  twistLimitDegrees=45
  twistMotion=Free
 }
 {
  bone1Name=dyn_loincloth_bk_l_MSTR
  bone2Name=pelvis
  swing1LimitDegrees=45
  swing1Motion=Free
  swing2LimitDegrees=45
  swing2Motion=Free
  twistLimitDegrees=45
  twistMotion=Free
 }
 {
  bone1Name=dyn_loincloth_bk_MSTR
  bone2Name=pelvis
  swing1LimitDegrees=45
  swing1Motion=Free
  swing2LimitDegrees=45
  swing2Motion=Free
  twistLimitDegrees=45
  twistMotion=Free
 }
 {
  bone1Name=dyn_loincloth_bk_r_MSTR
  bone2Name=pelvis
  swing1LimitDegrees=45
  swing1Motion=Free
  swing2LimitDegrees=45
  swing2Motion=Free
  twistLimitDegrees=45
  twistMotion=Free
 }
 {
  bone1Name=dyn_loincloth_bk_r_02
  bone2Name=dyn_loincloth_bk_03
  swing1LimitDegrees=45
  swing1Motion=Free
  swing2LimitDegrees=45
  swing2Motion=Free
  twistLimitDegrees=45
  twistMotion=Free
 }
 {
  bone1Name=dyn_loincloth_bk_l_02
  bone2Name=dyn_loincloth_bk_03
  swing1LimitDegrees=45
  swing1Motion=Free
  swing2LimitDegrees=45
  swing2Motion=Free
  twistLimitDegrees=45
  twistMotion=Free
 }
 {
  bone1Name=pelvis
  bone2Name=root
  swing1LimitDegrees=45
  swing1Motion=Free
  swing2LimitDegrees=45
  swing2Motion=Free
  twistLimitDegrees=45
  twistMotion=Free
 }
 {
  bone1Name=foot_r
  bone2Name=calf_r
  swing1LimitDegrees=45
  swing1Motion=Free
  swing2LimitDegrees=45
  swing2Motion=Free
  twistLimitDegrees=45
  twistMotion=Free
 }
 {
  bone1Name=foot_l
  bone2Name=calf_l
  swing1LimitDegrees=45
  swing1Motion=Free
  swing2LimitDegrees=45
  swing2Motion=Free
  twistLimitDegrees=45
  twistMotion=Free
 }
]
