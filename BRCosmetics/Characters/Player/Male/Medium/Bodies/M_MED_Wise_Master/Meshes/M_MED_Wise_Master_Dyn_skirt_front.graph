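# Physics asset for the front loincloth skirt of the Wise Master medium male player body, defining its dynamic bone chains and body collision.
# Three five-bone chains of boxes and capsules hang from small master bodies: center on the pelvis, left and right on the thighs, with limited swing and twist.
# Pelvis, spine, leg and foot capsules joined by free-rotation constraints form the body, with extra offset capsules on the thighs.
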PhysicsAsset /BRCosmetics/Characters/Player/Male/Medium/Bodies/M_MED_Wise_Master/Meshes/M_MED_Wise_Master_Dyn_skirt_front
bodies {
 calf_l [
  {shapeType=Capsule, length=9.3087, radius=12.7759, center=<-9.3515 -1.2662 -1.2704>, rotation=<rot 90 0 -2.7167>}
  {shapeType=Capsule, length=30.9558, radius=6.8579, center=<-35.0988 -0.1083 -0.6595>, rotation=<rot 90 0 3.9547>}
 ]
 calf_r [
  {shapeType=Capsule, length=9.3087, radius=12.7759, center=<9.352 1.2661 1.2765>, rotation=<rot 90 0 -2.7167>}
  {shapeType=Capsule, length=30.9558, radius=6.8579, center=<35.0988 0.1082 0.6595>, rotation=<rot 90 0 3.9547>}
 ]
 dyn_loincloth_fr_01 [{shapeType=Box, extentX=9.0138, extentY=21.4373, extentZ=2.1792, center=<3.6968 0 0.0004>}]
 dyn_loincloth_fr_02 [{shapeType=Box, extentX=9.8888, extentY=21.4373, extentZ=2.1792, center=<4.9427 0 0.0004>}]
 dyn_loincloth_fr_03 [{shapeType=Box, extentX=11.0138, extentY=21.4373, extentZ=3.1167, center=<5.1537 -0.0001 0.0004>, rotation=<rot 0 0 12.6026>}]
 dyn_loincloth_fr_04 [{shapeType=Capsule, length=13.1976, radius=6.5168, shapeName=dyn_loincloth_fr_04_capsule, center=<5 0 2>, rotation=<rot 90 -172.875 -82.875>}]
 dyn_loincloth_fr_05 [{shapeType=Capsule, length=13.1976, radius=6.5168, shapeName=dyn_loincloth_fr_05_capsule, center=<5 0 2>, rotation=<rot 90 -172.875 -82.875>}]
 dyn_loincloth_fr_MSTR [{shapeType=Capsule, length=1, radius=1, rotation=<rot 90 0 0>}]
 dyn_loincloth_fr_l_01 [{shapeType=Box, extentX=9.0138, extentY=13.9373, extentZ=2.1792, center=<3.6968 3.5 0.0006>}]
 dyn_loincloth_fr_l_02 [{shapeType=Capsule, length=10, radius=4.75, shapeName=dyn_loincloth_fr_l_02_capsule, center=<4.6639 3.7 2.5>, rotation=<rot 0 0 89.9999>}]
 dyn_loincloth_fr_l_03 [{shapeType=Capsule, length=10, radius=4.75, shapeName=dyn_loincloth_fr_l_03_capsule, center=<4.6639 3.8 2.5>, rotation=<rot 0 0 89.9999>}]
 dyn_loincloth_fr_l_04 [{
   shapeType=Capsule
   length=15
   radius=4.7144
   shapeName=dyn_loincloth_fr_l_04_capsule
   center=<5 3.5 2>
   rotation=<rot 90 -179.9209 -179.9209>
  }]
 dyn_loincloth_fr_l_05 [{
   shapeType=Capsule
   length=15
   radius=4.7144
   shapeName=dyn_loincloth_fr_l_05_capsule
   center=<5 3.5 2>
   rotation=<rot 90 -179.9604 -179.9605>
  }]
 dyn_loincloth_fr_l_MSTR [{shapeType=Capsule, length=1, radius=1, rotation=<rot 90 0 0>}]
 dyn_loincloth_fr_r_01 [{shapeType=Box, extentX=9.0138, extentY=13.9373, extentZ=2.1792, center=<3.6968 -3.5 0.0004>}]
 dyn_loincloth_fr_r_02 [{shapeType=Box, extentX=11.8888, extentY=13.9373, extentZ=5.5227, center=<4.9427 -3.5 2>, rotation=<rot 0 0 5.6183>}]
 dyn_loincloth_fr_r_03 [{shapeType=Box, extentX=11.0138, extentY=13.9373, extentZ=4.5227, center=<5.1537 -3.5 2>}]
 dyn_loincloth_fr_r_04 [{shapeType=Capsule, length=15, radius=4.7144, shapeName=dyn_loincloth_fr_r_04_capsule, center=<5 -3.5 2>, rotation=<rot 90 179.789 179.789>}]
 dyn_loincloth_fr_r_05 [{shapeType=Capsule, length=15, radius=4.7144, shapeName=dyn_loincloth_fr_r_05_capsule, center=<5 -3.5 2>, rotation=<rot 90 179.789 179.789>}]
 dyn_loincloth_fr_r_MSTR [{shapeType=Capsule, length=1, radius=1, rotation=<rot 90 0 0>}]
 foot_l [{shapeType=Capsule, length=16.5051, radius=7.6131, center=<-6.6239 10.2131 0.2607>, rotation=<rot -2.184 -174.3694 -87.6997>}]
 foot_r [{shapeType=Capsule, length=16.5051, radius=7.6131, center=<6.6239 -10.2131 -0.2607>, rotation=<rot -2.184 -174.3694 -87.6997>}]
 pelvis [{shapeType=Capsule, length=8.5661, radius=10.6545, center=<-7.6524 -0.3052 0>}]
 spine_04 [{shapeType=Capsule, length=13.9062, radius=10.7812, center=<0 7.1125 0>}]
 spine_05 [{shapeType=Capsule, length=10.9375, radius=13.75, center=<0.0213 7.1125 0>}]
 thigh_l [
  {shapeType=Capsule, length=24.2357, radius=10.9915, center=<-19.9223 -0.5025 -0.6706>, rotation=<rot 86.6156 -179.9999 -176.7087>}
  {shapeType=Capsule, length=22, radius=15, shapeName=thigh_l_capsule, center=<-26.8544 2.8302 -18.9129>, rotation=<rot -8.5241 0 -43.6024>}
 ]
 thigh_r [
  {shapeType=Capsule, length=24.2357, radius=10.9915, center=<19.9272 0.5025 0.5775>, rotation=<rot 86.6156 -179.9999 -176.7087>}
  {shapeType=Capsule, length=29.5, radius=15, shapeName=thigh_r_capsule, center=<25.189 1.1409 16.0605>, rotation=<rot 0 0 -48.9219>}
 ]
}
constraints [
 {bone1Name=dyn_loincloth_fr_l_01, bone2Name=dyn_loincloth_fr_l_MSTR, swing1LimitDegrees=80, swing2LimitDegrees=30, twistLimitDegrees=20}
 {bone1Name=dyn_loincloth_fr_l_02, bone2Name=dyn_loincloth_fr_l_01, swing1LimitDegrees=60, swing2LimitDegrees=30, twistLimitDegrees=20}
 {bone1Name=dyn_loincloth_fr_l_03, bone2Name=dyn_loincloth_fr_l_02, swing1LimitDegrees=60, swing2LimitDegrees=30, twistLimitDegrees=20}
 {bone1Name=dyn_loincloth_fr_l_04, bone2Name=dyn_loincloth_fr_l_03, swing1LimitDegrees=60, swing2LimitDegrees=30, twistLimitDegrees=20}
 {bone1Name=dyn_loincloth_fr_r_01, bone2Name=dyn_loincloth_fr_r_MSTR, swing1LimitDegrees=40, swing2LimitDegrees=30, twistLimitDegrees=20}
 {bone1Name=dyn_loincloth_fr_r_02, bone2Name=dyn_loincloth_fr_r_01, swing1LimitDegrees=45, swing2LimitDegrees=30, twistLimitDegrees=20}
 {bone1Name=dyn_loincloth_fr_r_03, bone2Name=dyn_loincloth_fr_r_02, swing1LimitDegrees=45, swing2LimitDegrees=30, twistLimitDegrees=20}
 {bone1Name=dyn_loincloth_fr_r_04, bone2Name=dyn_loincloth_fr_r_03, swing1LimitDegrees=45, swing2LimitDegrees=30, twistLimitDegrees=20}
 {bone1Name=dyn_loincloth_fr_01, bone2Name=dyn_loincloth_fr_MSTR, swing1LimitDegrees=45, swing2LimitDegrees=45, twistLimitDegrees=20}
 {bone1Name=dyn_loincloth_fr_02, bone2Name=dyn_loincloth_fr_01, swing1LimitDegrees=45, swing2LimitDegrees=45, twistLimitDegrees=20}
 {bone1Name=dyn_loincloth_fr_03, bone2Name=dyn_loincloth_fr_02, swing1LimitDegrees=45, swing2LimitDegrees=45, twistLimitDegrees=20}
 {bone1Name=dyn_loincloth_fr_04, bone2Name=dyn_loincloth_fr_03, swing1LimitDegrees=45, swing2LimitDegrees=45, twistLimitDegrees=20}
 {
  bone1Name=dyn_loincloth_fr_l_MSTR
  bone2Name=thigh_l
  swing1LimitDegrees=45
  swing1Motion=Free
  swing2LimitDegrees=45
  swing2Motion=Free
  twistLimitDegrees=45
  twistMotion=Free
 }
 {
  bone1Name=calf_l
  bone2Name=thigh_l
  swing1LimitDegrees=45
  swing1Motion=Free
  swing2LimitDegrees=45
  swing2Motion=Free
  twistLimitDegrees=45
  twistMotion=Free
 }
 {
  bone1Name=dyn_loincloth_fr_r_MSTR
  bone2Name=thigh_r
  swing1LimitDegrees=45
  swing1Motion=Free
  swing2LimitDegrees=45
  swing2Motion=Free
  twistLimitDegrees=45
  twistMotion=Free
 }
 {
  bone1Name=calf_r
  bone2Name=thigh_r
  swing1LimitDegrees=45
  swing1Motion=Free
  swing2LimitDegrees=45
  swing2Motion=Free
  twistLimitDegrees=45
  twistMotion=Free
 }
 {
  bone1Name=thigh_l
  bone2Name=pelvis
  swing1LimitDegrees=45
  swing1Motion=Free
  swing2LimitDegrees=45
  swing2Motion=Free
  twistLimitDegrees=45
  twistMotion=Free
 }
 {
  bone1Name=thigh_r
  bone2Name=pelvis
  swing1LimitDegrees=45
  swing1Motion=Free
  swing2LimitDegrees=45
  swing2Motion=Free
  twistLimitDegrees=45
  twistMotion=Free
 }
 {
  bone1Name=dyn_loincloth_fr_MSTR
  bone2Name=pelvis
  swing1LimitDegrees=45
  swing1Motion=Free
  swing2LimitDegrees=45
  swing2Motion=Free
  twistLimitDegrees=45
  twistMotion=Free
 }
 {
  bone1Name=foot_r
  bone2Name=calf_r
  swing1LimitDegrees=45
  swing1Motion=Free
  swing2LimitDegrees=45
  swing2Motion=Free
  twistLimitDegrees=45
  twistMotion=Free
 }
 {
  bone1Name=foot_l
  bone2Name=calf_l
  swing1LimitDegrees=45
  swing1Motion=Free
  swing2LimitDegrees=45
  swing2Motion=Free
  twistLimitDegrees=45
  twistMotion=Free
 }
 {bone1Name=dyn_loincloth_fr_05, bone2Name=dyn_loincloth_fr_04, swing1LimitDegrees=45, swing2LimitDegrees=45, twistLimitDegrees=20}
 {bone1Name=dyn_loincloth_fr_l_05, bone2Name=dyn_loincloth_fr_l_04, swing1LimitDegrees=60, swing2LimitDegrees=30, twistLimitDegrees=20}
 {bone1Name=dyn_loincloth_fr_r_05, bone2Name=dyn_loincloth_fr_r_04, swing1LimitDegrees=45, swing2LimitDegrees=30, twistLimitDegrees=20}
 {
  bone1Name=spine_05
  bone2Name=spine_04
  swing1LimitDegrees=45
  swing1Motion=Free
  swing2LimitDegrees=45
  swing2Motion=Free
  twistLimitDegrees=45
  twistMotion=Free
 }
 {
  bone1Name=spine_04
  bone2Name=pelvis
  swing1LimitDegrees=45
  swing1Motion=Free
  swing2LimitDegrees=45
  swing2Motion=Free
  twistLimitDegrees=45
  twistMotion=Free
 }
]
