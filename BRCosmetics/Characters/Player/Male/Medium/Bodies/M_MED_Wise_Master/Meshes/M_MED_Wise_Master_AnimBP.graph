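# Post-copy AnimBlueprint for the Wise Master body that animates its loincloth, hand trailers and half-thigh bones over the parent mesh pose.
# A cached pose applies springs, constraints and thigh-yaw correctives, then feeds two pelvis-space RigidBody sims for skirt and front.
# When rigid bodies are disabled or LOD exceeds 1, leg-driven constraints replace simulation.
# EventGraph adjusts alphas and gravity for sprinting, skydiving and certain vehicle drivers, whose loincloth shrinks.
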AnimBlueprint /BRCosmetics/Characters/Player/Male/Medium/Bodies/M_MED_Wise_Master/Meshes/M_MED_Wise_Master_AnimBP
target_skeleton=&/M_MED_Wise_Master_Skeleton
graphs [
 AnimGraph {nodes [
   "Constraint - dyn_loincloth_mid_l_MSTR (PPP)"!(ComponentPose, Alpha=1, ConstraintWeights_0=0.1761, ConstraintWeights_1:float=0.062, ConstraintWeights_2:float=0.7619) -> (Pose) {
    alphaBoolBlend {blendOption=Linear}
    boneToModify {boneName=dyn_loincloth_mid_l_MSTR}
    constraintSetup [{targetBone {boneName=spine_01}}, {targetBone {boneName=thigh_twist_01_l}}, {targetBone {boneName=pelvis}}]
    constraintWeights [1,1,1]
   }
   "Constraint - dyn_loincloth_mid_r_MSTR (PPP)"!(ComponentPose, Alpha=1, ConstraintWeights_0=0.1761, ConstraintWeights_1:float=0.062, ConstraintWeights_2:float=0.7619) -> (Pose) {
    alphaBoolBlend {blendOption=Linear}
    boneToModify {boneName=dyn_loincloth_mid_r_MSTR}
    constraintSetup [{targetBone {boneName=spine_01}}, {targetBone {boneName=thigh_twist_01_r}}, {targetBone {boneName=pelvis}}]
    constraintWeights [1,1,1]
   }
   "Transform (Modify) Bone - Bone: dyn_loincloth_mid_r_MSTR"(Alpha=1, Translation="-4.968,0.195,3.965", Rotation="0,0,0", Scale="1,1,1") {
    rotationMode=BMM_Additive
    rotationSpace=BCS_BoneSpace
    translationMode=BMM_Additive
    translationSpace=BCS_BoneSpace
    alphaBoolBlend {blendOption=Linear}
    boneToModify {boneName=dyn_loincloth_mid_r_MSTR}
    translation=<-4.968 0.195 3.965>
   }
   "Transform (Modify) Bone - Bone: dyn_loincloth_mid_r_MSTR"(Alpha=1, Translation="-4.729,0.000,-0.475", Rotation="0,0,0", Scale="1,1,1") {
    rotationMode=BMM_Additive
    rotationSpace=BCS_BoneSpace
    translationMode=BMM_Additive
    translationSpace=BCS_BoneSpace
    alphaBoolBlend {blendOption=Linear}
    alphaScaleBiasClamp {bInterpResult=true}
    boneToModify {boneName=dyn_loincloth_mid_r_MSTR}
    translation {x=-4.729, z=-0.475}
   }
   "Transform (Modify) Bone - Bone: dyn_loincloth_mid_l_MSTR"(Alpha=1, Translation="-5.286,-0.000,4.828", Rotation="0,0,0", Scale="1,1,1") {
    rotationMode=BMM_Additive
    rotationSpace=BCS_BoneSpace
    translationMode=BMM_Additive
    translationSpace=BCS_BoneSpace
    alphaBoolBlend {blendOption=Linear}
    alphaScaleBiasClamp {bInterpResult=true}
    boneToModify {boneName=dyn_loincloth_mid_l_MSTR}
    translation {x=-5.286, z=4.828}
   }
   n1 = Break Fort Anim Input Skydiving External Force(FortAnimInput_SkydivingExternalForce=Get SkydivingExternalForceInput()) -> (
    bUseSkydivingVectorForce=~
    SkydivingVectorForce=~
    HeadToPelvisDirection=~
    FloatingMultiplier=~
    FloatingAdditive=~
    DivingMultiplier=~
    DivingAdditive=~
    ParachutingMultiplier=~
    ParachutingAdditive=~
   )
   n2 = RigidBody!(
    ComponentPose=Local To Component(LocalPose=Use cached pose 'startCache'())
    Alpha=Get in vehicle rigid alpha() -> (in vehicle rigid alpha:double)
    OverrideWorldGravity:Vector=Get gravity_override() -> (gravity_override:Vector)
    ExternalForce=@n1.SkydivingVectorForce
   ) -> (Pose) {
    bClampLinearTranslationLimitToRefPose=true
    bOverrideWorldGravity=true
    overridePhysicsAsset=&/M_MED_Wise_Master_Dyn_skirt
    simulationSpace=BaseBoneSpace
    alphaBoolBlend {blendOption=Linear}
    baseBoneRef {boneName=pelvis}
    componentAppliedLinearAccClamp=<3000 3000 3000>
    componentLinearAccScale=<0.45 0.3 0.3>
    componentLinearVelScale=<0.45 0.3 0.3>
    overrideWorldGravity {z=-980}
   }
   n3 = Get pawnVelocityAlpha_inv() -> (pawnVelocityAlpha_inv:double)
   n4 = "Constraint - dyn_loincloth_fr_r_MSTR (R)"(ComponentPose=@n2, Alpha=@n3, ConstraintWeights_0=1) {
    alphaBoolBlend {blendOption=Linear}
    boneToModify {boneName=dyn_loincloth_fr_r_MSTR}
    constraintSetup [{transformType=Rotation, targetBone {boneName=thigh_r}}]
    constraintWeights [1]
   }
   n5 = "Constraint - dyn_loincloth_fr_MSTR (RR)"!(ComponentPose=@n4, Alpha=@n3, ConstraintWeights_0=0.5, ConstraintWeights_1:float=0.5) -> (Pose) {
    alphaBoolBlend {blendOption=Linear}
    boneToModify {boneName=dyn_loincloth_fr_MSTR}
    constraintSetup [{transformType=Rotation, targetBone {boneName=thigh_l}}, {transformType=Rotation, targetBone {boneName=thigh_r}}]
    constraintWeights [1,1]
   }
   n6 = "Constraint - dyn_loincloth_fr_l_MSTR (R)"(ComponentPose=@n5, Alpha=@n3, ConstraintWeights_0=1) {
    alphaBoolBlend {blendOption=Linear}
    boneToModify {boneName=dyn_loincloth_fr_l_MSTR}
    constraintSetup [{transformType=Rotation, targetBone {boneName=thigh_l}}]
    constraintWeights [1]
   }
   n7 = RigidBody!(ComponentPose=@n6, Alpha=Get pawnVelocityAlpha() -> (pawnVelocityAlpha:double), ExternalForce=@n1.SkydivingVectorForce) -> (Pose) {
    bClampLinearTranslationLimitToRefPose=true
    bOverrideWorldGravity=true
    overridePhysicsAsset=&/M_MED_Wise_Master_Dyn_skirt_front
    simulationSpace=BaseBoneSpace
    alphaBoolBlend {blendOption=Linear}
    alphaScaleBiasClamp {bInterpResult=true}
    baseBoneRef {boneName=pelvis}
    componentAppliedLinearAccClamp=<3000 3000 3000>
    componentLinearAccScale=<0.45 0.4 0.4>
    componentLinearVelScale=<0.45 0.4 0.4>
    overrideWorldGravity {z=-980}
   }
   n8 = "Transform (Modify) Bone - Bone: dyn_loincloth_bk_l_MSTR"!(ComponentPose=@n7, Alpha=1, Translation="0,0,0", Rotation="15,0,0") -> (Pose) {
    rotationMode=BMM_Replace
    rotationSpace=BCS_BoneSpace
    scaleSpace=BCS_BoneSpace
    translationMode=BMM_Additive
    translationSpace=BCS_BoneSpace
    alphaBoolBlend {blendOption=Linear}
    alphaScaleBiasClamp {bInterpResult=true}
    boneToModify {boneName=dyn_loincloth_bk_l_MSTR}
    rotation {pitch=15}
    scale=<0.1 0.1 0.1>
   }
   n9 = "Transform (Modify) Bone - Bone: dyn_loincloth_bk_MSTR"!(ComponentPose=@n8, Alpha=1, Translation="0,0,0", Rotation="10,0,0") -> (Pose) {
    rotationMode=BMM_Replace
    rotationSpace=BCS_BoneSpace
    scaleSpace=BCS_BoneSpace
    translationMode=BMM_Additive
    translationSpace=BCS_BoneSpace
    alphaBoolBlend {blendOption=Linear}
    boneToModify {boneName=dyn_loincloth_bk_MSTR}
    rotation {pitch=10}
    scale=<0.1 0.1 0.1>
   }
   n10 = "Transform (Modify) Bone - Bone: dyn_loincloth_bk_r_MSTR"!(ComponentPose=@n9, Alpha=1, Translation="0,0,0", Rotation="15,0,0") -> (Pose) {
    rotationMode=BMM_Replace
    rotationSpace=BCS_BoneSpace
    scaleSpace=BCS_BoneSpace
    translationMode=BMM_Additive
    translationSpace=BCS_BoneSpace
    alphaBoolBlend {blendOption=Linear}
    alphaScaleBiasClamp {bInterpResult=true}
    boneToModify {boneName=dyn_loincloth_bk_r_MSTR}
    rotation {pitch=15}
    scale=<0.1 0.1 0.1>
   }
   n11 = "Constraint - dyn_loincloth_bk_l_02 (P)"(ComponentPose=Local To Component(LocalPose=Use cached pose 'startCache'()), Alpha=1, ConstraintWeights_0=1) {
    alphaBoolBlend {blendOption=Linear}
    boneToModify {boneName=dyn_loincloth_bk_l_02}
    constraintSetup [{targetBone {boneName=thigh_l}}]
    constraintWeights [1]
   }
   n12 = "Constraint - dyn_loincloth_bk_r_02 (P)"(ComponentPose=@n11, Alpha=1, ConstraintWeights_0=1) {
    alphaBoolBlend {blendOption=Linear}
    boneToModify {boneName=dyn_loincloth_bk_r_02}
    constraintSetup [{targetBone {boneName=thigh_r}}]
    constraintWeights [1]
   }
   n13 = "Constraint - dyn_loincloth_bk_02 (PP)"!(ComponentPose=@n12, Alpha=1, ConstraintWeights_0=0.5, ConstraintWeights_1:float=0.5) -> (Pose) {
    alphaBoolBlend {blendOption=Linear}
    alphaScaleBiasClamp {bClampResult=true, clampMax=0.5}
    boneToModify {boneName=dyn_loincloth_bk_02}
    constraintSetup [{targetBone {boneName=dyn_loincloth_bk_l_02}}, {targetBone {boneName=dyn_loincloth_bk_r_02}}]
    constraintWeights [1,1]
   }
   n14 = "Constraint - dyn_loincloth_bk_l_04 (P)"(ComponentPose=@n13, Alpha=1, ConstraintWeights_0=1) {
    alphaBoolBlend {blendOption=Linear}
    boneToModify {boneName=dyn_loincloth_bk_l_04}
    constraintSetup [{targetBone {boneName=calf_l}}]
    constraintWeights [1]
   }
   n15 = "Constraint - dyn_loincloth_bk_r_04 (P)"(ComponentPose=@n14, Alpha=1, ConstraintWeights_0=1) {
    alphaBoolBlend {blendOption=Linear}
    boneToModify {boneName=dyn_loincloth_bk_r_04}
    constraintSetup [{targetBone {boneName=calf_r}}]
    constraintWeights [1]
   }
   n16 = "Constraint - dyn_loincloth_bk_04 (PP)"!(ComponentPose=@n15, Alpha=1, ConstraintWeights_0=0.5, ConstraintWeights_1:float=0.5) -> (Pose) {
    alphaBoolBlend {blendOption=Linear}
    alphaScaleBiasClamp {bClampResult=true, clampMax=0.5}
    boneToModify {boneName=dyn_loincloth_bk_04}
    constraintSetup [{targetBone {boneName=dyn_loincloth_bk_l_04}}, {targetBone {boneName=dyn_loincloth_bk_r_04}}]
    constraintWeights [1,1]
   }
   n17 = "Constraint - dyn_loincloth_fr_l_01 (P)"(ComponentPose=@n16, Alpha=1, ConstraintWeights_0=1) {
    alphaBoolBlend {blendOption=Linear}
    boneToModify {boneName=dyn_loincloth_fr_l_01}
    constraintSetup [{targetBone {boneName=thigh_l}}]
    constraintWeights [1]
   }
   n18 = "Constraint - dyn_loincloth_fr_r_01 (P)"(ComponentPose=@n17, Alpha=1, ConstraintWeights_0=1) {
    alphaBoolBlend {blendOption=Linear}
    boneToModify {boneName=dyn_loincloth_fr_r_01}
    constraintSetup [{targetBone {boneName=thigh_r}}]
    constraintWeights [1]
   }
   n19 = "Constraint - dyn_loincloth_fr_01 (PP)"!(ComponentPose=@n18, Alpha=1, ConstraintWeights_0=0.5, ConstraintWeights_1:float=0.5) -> (Pose) {
    alphaBoolBlend {blendOption=Linear}
    alphaScaleBiasClamp {bClampResult=true}
    boneToModify {boneName=dyn_loincloth_fr_01}
    constraintSetup [{targetBone {boneName=dyn_loincloth_fr_l_01}}, {targetBone {boneName=dyn_loincloth_fr_r_01}}]
    constraintWeights [1,1]
   }
   n20 = "Transform (Modify) Bone - Bone: dyn_loincloth_bk_l_MSTR"(ComponentPose=@n19, Alpha=1, Translation="0,0,2", Rotation="5,0,0") {
    rotationMode=BMM_Replace
    rotationSpace=BCS_BoneSpace
    scaleSpace=BCS_BoneSpace
    translationMode=BMM_Additive
    translationSpace=BCS_BoneSpace
    alphaBoolBlend {blendOption=Linear}
    boneToModify {boneName=dyn_loincloth_bk_l_MSTR}
    rotation {pitch=5}
    scale=<0.1 0.1 0.1>
    translation {z=2}
   }
   n21 = "Transform (Modify) Bone - Bone: dyn_loincloth_bk_MSTR"(ComponentPose=@n20, Alpha=1, Translation="0,0,2", Rotation="10,0,0") {
    rotationMode=BMM_Replace
    rotationSpace=BCS_BoneSpace
    scaleSpace=BCS_BoneSpace
    translationMode=BMM_Additive
    translationSpace=BCS_BoneSpace
    alphaBoolBlend {blendOption=Linear}
    boneToModify {boneName=dyn_loincloth_bk_MSTR}
    rotation {pitch=10}
    scale=<0.1 0.1 0.1>
    translation {z=2}
   }
   n22 = "Transform (Modify) Bone - Bone: dyn_loincloth_bk_r_MSTR"(ComponentPose=@n21, Alpha=1, Translation="0,0,2", Rotation="5,0,0") {
    rotationMode=BMM_Replace
    rotationSpace=BCS_BoneSpace
    scaleSpace=BCS_BoneSpace
    translationMode=BMM_Additive
    translationSpace=BCS_BoneSpace
    alphaBoolBlend {blendOption=Linear}
    boneToModify {boneName=dyn_loincloth_bk_r_MSTR}
    rotation {pitch=5}
    scale=<0.1 0.1 0.1>
    translation {z=2}
   }
   n23 = "Transform (Modify) Bone - Bone: dyn_loincloth_bk_l_03"!(ComponentPose=@n22, Alpha=1, Translation="0,0,0", Rotation="5,0,0") -> (Pose) {
    rotationMode=BMM_Replace
    rotationSpace=BCS_BoneSpace
    scaleSpace=BCS_BoneSpace
    translationMode=BMM_Additive
    translationSpace=BCS_BoneSpace
    alphaBoolBlend {blendOption=Linear}
    boneToModify {boneName=dyn_loincloth_bk_l_03}
    rotation {pitch=5}
    scale=<0.1 0.1 0.1>
   }
   n24 = "Transform (Modify) Bone - Bone: dyn_loincloth_bk_03"!(ComponentPose=@n23, Alpha=1, Translation="0,0,0", Rotation="15.000,-0.000,0.000") -> (Pose) {
    rotationMode=BMM_Replace
    rotationSpace=BCS_BoneSpace
    scaleSpace=BCS_BoneSpace
    translationMode=BMM_Additive
    translationSpace=BCS_BoneSpace
    alphaBoolBlend {blendOption=Linear}
    boneToModify {boneName=dyn_loincloth_bk_03}
    rotation=<rot 15 0 0>
    scale=<0.1 0.1 0.1>
   }
   n25 = "Transform (Modify) Bone - Bone: dyn_loincloth_bk_r_03"!(ComponentPose=@n24, Alpha=1, Translation="0,0,0", Rotation="5,0,0") -> (Pose) {
    rotationMode=BMM_Replace
    rotationSpace=BCS_BoneSpace
    scaleSpace=BCS_BoneSpace
    translationMode=BMM_Additive
    translationSpace=BCS_BoneSpace
    alphaBoolBlend {blendOption=Linear}
    boneToModify {boneName=dyn_loincloth_bk_r_03}
    rotation {pitch=5}
    scale=<0.1 0.1 0.1>
   }
   Output Pose(Result=Blend Poses by bool(
     BlendPose_0=Component To Local(ComponentPose=@n25)
     BlendPose_1=Component To Local(ComponentPose=@n10)
     bActiveValue=Get rigids_off() -> (rigids_off:bool)
    ) {blendType=Linear}) {layerGroup=None}
   n26 = Local To Component(LocalPose=Copy Pose From Mesh() {bUseAttachedParent=true})
   n27 = Get thigh_half_value() -> (thigh_half_value:double)
   n28 = "Constraint - thigh_l_half (R)"(
    ComponentPose=Spring controller - Bone: hand_r_trailer_02!(
     ComponentPose:Component Space Pose Link Structure=Spring controller - Bone: hand_l_trailer_02!(ComponentPose:Component Space Pose Link Structure=@n26, Alpha:float=1) -> (Pose:Component Space Pose Link Structure) {is=title AnimGraphNode_SpringBone, springDamping=5, alphaBoolBlend {blendOption=Linear}, springBone {boneName=hand_l_trailer_02}}
     Alpha:float=1
    ) -> (Pose:Component Space Pose Link Structure) {
     is=title AnimGraphNode_SpringBone
     springDamping=8
     springStiffness=100
     alphaBoolBlend {blendOption=Linear}
     springBone {boneName=hand_r_trailer_02}
    }
    Alpha=@n27
    ConstraintWeights_0=1
   ) {
    alphaBoolBlend {blendOption=Linear}
    boneToModify {boneName=thigh_l_half}
    constraintSetup [{transformType=Rotation, targetBone {boneName=thigh_l}}]
    constraintWeights [1]
   }
   n29 = "Constraint - thigh_r_half (R)"(ComponentPose=@n28, Alpha=@n27, ConstraintWeights_0=1) {
    alphaBoolBlend {blendOption=Linear}
    boneToModify {boneName=thigh_r_half}
    constraintSetup [{transformType=Rotation, targetBone {boneName=thigh_r}}]
    constraintWeights [1]
   }
   n30 = Get in vehicle scale alpha() -> (in vehicle scale alpha:double)
   n31 = "Transform (Modify) Bone - Bone: dyn_loincloth_fr_l_MSTR"(ComponentPose=@n29, Alpha=@n30, Translation="0,0,0", Rotation="0,0,0", Scale="0.1,0.1,0.1") {scaleMode=BMM_Replace, alphaBoolBlend {blendOption=Linear}, boneToModify {boneName=dyn_loincloth_fr_l_MSTR}, scale=<0.1 0.1 0.1>}
   n32 = "Transform (Modify) Bone - Bone: dyn_loincloth_fr_r_MSTR"(ComponentPose=@n31, Alpha=@n30, Translation="0,0,0", Rotation="0,0,0", Scale="0.1,0.1,0.1") {scaleMode=BMM_Replace, alphaBoolBlend {blendOption=Linear}, boneToModify {boneName=dyn_loincloth_fr_r_MSTR}, scale=<0.1 0.1 0.1>}
   n33 = "Transform (Modify) Bone - Bone: dyn_loincloth_fr_MSTR"(ComponentPose=@n32, Alpha=@n30, Translation="0,0,0", Rotation="0,0,0", Scale="0.1,0.1,0.1") {scaleMode=BMM_Replace, alphaBoolBlend {blendOption=Linear}, boneToModify {boneName=dyn_loincloth_fr_MSTR}, scale=<0.1 0.1 0.1>}
   n34 = "Transform (Modify) Bone - Bone: dyn_loincloth_bk_l_MSTR"!(ComponentPose=@n33, Alpha=@n30, Translation="0,0,0", Rotation="0,0,0", Scale:Vector="0.1,0.1,0.1") -> (Pose) {scaleMode=BMM_Replace, alphaBoolBlend {blendOption=Linear}, boneToModify {boneName=dyn_loincloth_bk_l_MSTR}, scale=<0.1 0.1 0.1>}
   n35 = "Transform (Modify) Bone - Bone: dyn_loincloth_bk_MSTR"!(ComponentPose=@n34, Alpha=@n30, Translation="0,0,0", Rotation="0,0,0", Scale:Vector="0.1,0.1,0.1") -> (Pose) {scaleMode=BMM_Replace, alphaBoolBlend {blendOption=Linear}, boneToModify {boneName=dyn_loincloth_bk_MSTR}, scale=<0.1 0.1 0.1>}
   n36 = "Transform (Modify) Bone - Bone: dyn_loincloth_bk_r_MSTR"!(ComponentPose=@n35, Alpha=@n30, Translation="0,0,0", Rotation="0,0,0", Scale:Vector="0.1,0.1,0.1") -> (Pose) {scaleMode=BMM_Replace, alphaBoolBlend {blendOption=Linear}, boneToModify {boneName=dyn_loincloth_bk_r_MSTR}, scale=<0.1 0.1 0.1>}
   n37 = "Constraint - dyn_loincloth_fr_r_MSTR (PPP)"!(ComponentPose=@n36, Alpha=1, ConstraintWeights_0=0.2167, ConstraintWeights_1:float=0.0841, ConstraintWeights_2:float=0.6992) -> (Pose) {
    alphaBoolBlend {blendOption=Linear}
    boneToModify {boneName=dyn_loincloth_fr_r_MSTR}
    constraintSetup [{targetBone {boneName=spine_01}}, {targetBone {boneName=thigh_twist_01_r}}, {targetBone {boneName=pelvis}}]
    constraintWeights [1,1,1]
   }
   n38 = "Constraint - dyn_loincloth_fr_MSTR (PPPP)"!(
    ComponentPose=@n37
    Alpha=1
    ConstraintWeights_0=0.1413
    ConstraintWeights_1:float=0.0194
    ConstraintWeights_2:float=0.0185
    ConstraintWeights_3:float=0.8207
   ) -> (Pose) {
    alphaBoolBlend {blendOption=Linear}
    boneToModify {boneName=dyn_loincloth_fr_MSTR}
    constraintSetup [
     {targetBone {boneName=spine_01}}
     {targetBone {boneName=thigh_twist_01_l}}
     {targetBone {boneName=thigh_twist_01_r}}
     {targetBone {boneName=pelvis}}
    ]
    constraintWeights [1,1,1,1]
   }
   n39 = "Constraint - dyn_loincloth_fr_l_MSTR (PPP)"!(ComponentPose=@n38, Alpha=1, ConstraintWeights_0=0.2074, ConstraintWeights_1:float=0.0954, ConstraintWeights_2:float=0.6972) -> (Pose) {
    alphaBoolBlend {blendOption=Linear}
    boneToModify {boneName=dyn_loincloth_fr_l_MSTR}
    constraintSetup [{targetBone {boneName=spine_01}}, {targetBone {boneName=thigh_twist_01_l}}, {targetBone {boneName=pelvis}}]
    constraintWeights [1,1,1]
   }
   n40 = Get thigh_rotation_l() -> (thigh_rotation_l:double)
   n41 = "Transform (Modify) Bone - Bone: dyn_loincloth_fr_l_MSTR"(ComponentPose=@n39, Alpha=@n40, Translation="-14.227,11.055,21.980", Rotation="0.005,-0.011,34.989", Scale="1,1,1") {
    rotationMode=BMM_Additive
    rotationSpace=BCS_BoneSpace
    translationMode=BMM_Additive
    translationSpace=BCS_BoneSpace
    alphaBoolBlend {blendOption=Linear}
    alphaScaleBiasClamp {bInterpResult=true}
    boneToModify {boneName=dyn_loincloth_fr_l_MSTR}
    rotation=<rot 0.005 -0.011 34.989>
    translation=<-14.227 11.055 21.98>
   }
   n42 = "Transform (Modify) Bone - Bone: dyn_loincloth_fr_MSTR"(ComponentPose=@n41, Alpha=@n40, Translation="0.000,-0.000,2.430", Rotation="0,0,0", Scale="1,1,1") {
    rotationMode=BMM_Additive
    rotationSpace=BCS_BoneSpace
    translationMode=BMM_Additive
    translationSpace=BCS_BoneSpace
    alphaBoolBlend {blendOption=Linear}
    alphaScaleBiasClamp {bInterpResult=true}
    boneToModify {boneName=dyn_loincloth_fr_MSTR}
    translation {z=2.43}
   }
   n43 = "Transform (Modify) Bone - Bone: dyn_loincloth_fr_r_MSTR"(ComponentPose=@n42, Alpha=@n40, Translation="-4.897,-3.299,3.994", Rotation="0,0,0", Scale="1,1,1") {
    rotationMode=BMM_Additive
    rotationSpace=BCS_BoneSpace
    translationMode=BMM_Additive
    translationSpace=BCS_BoneSpace
    alphaBoolBlend {blendOption=Linear}
    alphaScaleBiasClamp {bInterpResult=true}
    boneToModify {boneName=dyn_loincloth_fr_r_MSTR}
    rotation=<rot 0 0 0>
    translation=<-4.897 -3.299 3.994>
   }
   n44 = "Transform (Modify) Bone - Bone: dyn_loincloth_fr_r_MSTR"(
    ComponentPose=@n43
    Alpha=Get thigh rotation_r() -> (thigh rotation_r:double)
    Translation="-4,-1.773,7.753"
    Rotation="0.005,0.000,-18.596"
    Scale="1,1,1"
   ) {
    rotationMode=BMM_Additive
    rotationSpace=BCS_BoneSpace
    translationMode=BMM_Additive
    translationSpace=BCS_BoneSpace
    alphaBoolBlend {blendOption=Linear}
    boneToModify {boneName=dyn_loincloth_fr_r_MSTR}
    rotation=<rot 0.005 0 -18.596>
    translation=<-4 -1.773 7.753>
   }
   Save cached pose 'startCache'(Pose=Component To Local(ComponentPose=@n44)) {cacheName=startCache}
   Get Crouch Alpha() -> (Crouch Alpha:double)
   Get one() -> (one:double)
  ]}
 EventGraph {nodes [
   Event BlueprintUpdateAnimation()
   > n1 = Sequence!(execute) -> (then_0, then_1, then_2:Exec, then_3:Exec, then_4:Exec, then_5:Exec)
   > n2 = Branch(Condition=Get bIsOstrichVehicleDriver() -> (bIsOstrichVehicleDriver:bool))
   Set in vehicle rigid alpha(execute=[@n4.then, @n2.then, @n3.then], in vehicle rigid alpha:double=0) -> (Output_Get:double)
   > Set in vehicle scale alpha(in vehicle scale alpha:double=1) -> (Output_Get:double)
   n3 = Branch(execute=@n2.else, Condition=Get bIsFerretVehicleDriver() -> (bIsFerretVehicleDriver:bool))
   n4 = Branch(execute=@n3.else, Condition=Get bIsOctopusVehicleDriver() -> (bIsOctopusVehicleDriver:bool))
   Set in vehicle rigid alpha(execute=@n4.else, in vehicle rigid alpha=1)
   > Set in vehicle scale alpha(in vehicle scale alpha=0)
   n5 = Branch(execute=@n1.then_1, Condition=Get bIsSkyDiving())
   > n6 = Branch(Condition=Get bIsParachuteOpened())
   Set gravity_override(execute=[@n6.then, @n5.else], gravity_override="0, 0,-980")
   Set gravity_override(execute=@n6.else, gravity_override="0, 0,-300")
   n7 = Branch(execute=@n1.then_2, Condition=Get bIsSprinting())
   > Set thigh_half_value(thigh_half_value:double=1) -> (Output_Get:double)
   > Set pawnVelocityAlpha(pawnVelocityAlpha:double=0.5) -> (Output_Get:double)
   > Set pawnVelocityAlpha_inv(pawnVelocityAlpha_inv:double=1) -> (Output_Get:double)
   Set thigh_half_value(execute=@n7.else, thigh_half_value=0.8)
   > Set pawnVelocityAlpha(pawnVelocityAlpha=1)
   > Set pawnVelocityAlpha_inv(pawnVelocityAlpha_inv=0)
   n8 = Get PawnVelocity()
   n9 = Branch(
    execute=@n1.then_3
    Condition="float < float"!(A:double=Vector Length(A=@n8), B:double=2) -> (ReturnValue:bool) {is=title K2Node_CallFunction}
   )
   n10 = GetPartSkeletalMeshComponent()
   n11 = GetDeltaTransformFromRefPose(self=@n10, BoneName=hand_l_trailer_02, BaseName=hand_l_trailer_01)
   > Set hand_l movement(hand_l movement:Vector=vector * vector(A:Vector=BreakTransform(InTransform=@n11), B:Vector="1000,1000,1000") -> (ReturnValue:Vector) {is=title K2Node_CommutativeAssociativeBinaryOperator}) -> (Output_Get:Vector)
   n12 = GetDeltaTransformFromRefPose(self=@n10, BoneName=hand_r_trailer_02, BaseName=hand_r_trailer_01)
   > Set Hand_r movement(Hand_r movement:Vector=vector * vector(A=BreakTransform(InTransform=@n12), B="1000,1000,1000")) -> (Output_Get:Vector)
   Set hand_l movement(execute=@n9.else, hand_l movement="0, 0, 0")
   > Set Hand_r movement(Hand_r movement="0, 0, 0")
   n13 = BreakTransform(InTransform=GetDeltaTransformFromRefPose(self=GetOwnerSkeletalMeshComponent(), BoneName=thigh_r, BaseName=pelvis))
   Set thigh rotation_r(execute=@n1.then_4, thigh rotation_r:double=MapRangeClamped(Value=BreakRotator(InRot=@n13.Rotation).Yaw, InRangeA=-50, InRangeB=-90)) -> (Output_Get:double)
   n14 = BreakTransform(InTransform=GetDeltaTransformFromRefPose(self=GetOwnerSkeletalMeshComponent(), BoneName=thigh_l, BaseName=pelvis))
   > Set thigh_rotation_l(thigh_rotation_l:double=MapRangeClamped(Value=BreakRotator(InRot=@n14.Rotation).Yaw, InRangeA=-50, InRangeB=-90)) -> (Output_Get:double)
   n15 = Get CurrentLOD()
   n16 = Get bRigidBodyEnabled()
   Set rigids_off(execute=@n1.then_5, rigids_off:bool=OR Boolean(A=NOT Boolean(A=@n16), B="integer > integer"(A=@n15, B=1) {is=title K2Node_CallFunction})) -> (Output_Get:bool)
   Event OnBeginCrouching()
   > Set Crouch Alpha(Crouch Alpha:double=1) -> (Output_Get:double)
   > Set Crouch Alpha Inv(Crouch Alpha Inv:double=0) -> (Output_Get:double)
   Event OnEndCrouching()
   > Set Crouch Alpha(Crouch Alpha=0)
   > Set Crouch Alpha Inv(Crouch Alpha Inv=1)
  ]}
]
variables [
 Crouch Alpha {type=number}
 Crouch Alpha Inv {type=number, default=1}
 gravity_override {type=Vector, default=<0 0 -980>}
 hand_l movement {type=Vector}
 Hand_r movement {type=Vector}
 in vehicle rigid alpha {type=number}
 in vehicle scale alpha {type=number}
 one {type=number, default=1}
 pawnVelocityAlpha {type=number}
 pawnVelocityAlpha_inv {type=number, default=1}
 rigids_off {type=boolean}
 thigh rotation_r {type=number}
 thigh_half_value {type=number, default=0.8}
 thigh_rotation_l {type=number}
]
